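# M_EMGradient is a subsurface-shaded material that builds a vertical color gradient and is flagged for use on instanced static meshes.
# The LinearGradient function's vertical output passes through LinearTosRGB to serve as the shared alpha for two lerps.
# Both lerps start from the light blue vector parameter "v", one blending toward a darker blue parameter and the other toward a black constant.
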Material /ContentHall/Materials/M_EMGradient
graph {
 properties {bUsedWithInstancedStaticMeshes=true, pixelDepthOffsetMode=PDOM_Legacy, shadingModel=MSM_Subsurface, usedShadingModels=Subsurface}
 nodes [
  Constant()
  Constant() {r=1}
  n1 = VectorParameter() {parameterName=v, defaultValue=<rgba 0.1771 0.6827 1 1>}
  n2 = MaterialFunctionCall!(Linear=MaterialFunctionCall() {materialFunction=&LinearGradient}.VGradient) {materialFunction=/Engine/Functions/Engine_MaterialFunctions01/ImageAdjustment/LinearTosRGB}
  LinearInterpolate(A=@n1.RGB, B=VectorParameter() {defaultValue=<rgba 0.0586 0.1642 0.625 1>}.RGB, Alpha=@n2.sRGB)
  LinearInterpolate(A=@n1.RGB, B=Constant(), Alpha=@n2.sRGB)
 ]
}
